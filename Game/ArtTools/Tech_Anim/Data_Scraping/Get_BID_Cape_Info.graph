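# Editor utility widget for tech animation that exports a report of Fortnite backpack items (BIDs) identified as capes.
# Its button runs an embedded Python script that scans every AthenaBackpackItemDefinition in the Asset Registry and reads its Backpack character parts.
# Capes are detected by gameplay tag, "cape" attach or loot socket, or cloth-collision flag.
# Attachment, AnimBP, body and tag data go to a hardcoded D: drive CSV, while text blocks show status.
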EditorUtilityWidgetBlueprint /Game/ArtTools/Tech_Anim/Data_Scraping/Get_BID_Cape_Info
graphs [EventGraph {nodes [
   "On Clicked (Get_Info_Button)"() {is=title K2Node_ComponentBoundEvent}
   > Set IsRunning(IsRunning:bool=true) -> (Output_Get:bool)
   > Set Text(self=Get Info_Text() -> (Info_Text:Text Block Object Reference))
   > Set Text(self=Get File_Location_Text() -> (File_Location_Text:Text Block Object Reference))
   > Set Text(Text="LOCTEXT(\"Getting Info..\")", self=Get Get_Info_Status() -> (Get_Info_Status:Text Block Object Reference))
   > Execute Python Script(PythonScript="import unreal\r\nimport csv\r\n\r\n## Get CustomCharacterPat property/value\r\n## Source\\FortniteGame\\Private\\Customization\\CustomCharacterPart.cpp\r\n\r\n\r\nasset_registry = unreal.AssetRegistryHelpers.get_asset_registry()\r\n\r\ndef get_gamplay_tags(bid_asset):\r\n    \"\"\"\r\n    Get gameplay tags from containers on the BID\r\n    \"\"\"\r\n    asset_data = unreal.AssetData.get_asset(bid_asset)\r\n    tag_containers = [\"GameplayTags\", \"ExclusiveRequiresOutfitTags\"]\r\n    gameplay_tags = []\r\n    for tag_container_name in tag_containers:                            \r\n        try:    \r\n            container = asset_data.get_editor_property(tag_container_name)\r\n            for tag in container.gameplay_tags:\r\n                if tag.get_editor_property(\"tag_name\"):                                    \r\n                    tag_name = str(tag.get_editor_property(\"tag_name\"))\r\n                    gameplay_tags.append(tag_name)\r\n        except Exception:\r\n            print('  Failed to get property \"GameplayTags\"')            \r\n    \r\n    return gameplay_tags\r\n\r\ndef get_part_data(part_asset):\r\n    \"\"\"\r\n    Get Backpack Data from the CharacterPart Asset\r\n    \"\"\"\r\n    \r\n    backpack_info = {}\r\n    part_asset_data = unreal.AssetData.get_asset(part_asset)\r\n    if not part_asset_data:\r\n        return backpack_info\r\n    \r\n    try:            \r\n        anim_class = part_asset_data.get_editor_property(\"AdditionalData\").get_editor_property(\"AnimClass\")\r\n    except Exception:\r\n        anim_class = None        \r\n    if anim_class:            \r\n        anim_class_name = anim_class.get_name()\r\n        anim_class_name = str(anim_class_name)[:-2]            \r\n        backpack_info[\"AnimBP\"] = anim_class_name\r\n        ##print(\"  AnimBP: {}\".format(str(anim_class_name)))\r\n   \r\n    try:            \r\n        attach_socket = part_asset_data.get_editor_property(\"AdditionalData\").get_editor_property(\"AttachSocketName\")\r\n    except Exception:\r\n        attach_socket = None\r\n    if attach_socket:\r\n        if not attach_socket == \"None\":\r\n            backpack_info[\"AttachSocketName\"] = str(attach_socket).lower()\r\n            ##print(\"  AttachSocketName: {}\".format(attach_socket))\r\n    \r\n    try:            \r\n        loot_socket = part_asset_data.get_editor_property(\"AdditionalData\").get_editor_property(\"LootSocketName\")\r\n    except Exception:\r\n        loot_socket = None        \r\n    if loot_socket:\r\n        if not loot_socket == \"None\":\r\n            backpack_info[\"LootSocket\"] = str(loot_socket).lower()\r\n            ##print(\"  LootSocketName: {}\".format(loot_socket))  \r\n\r\n    try:            \r\n        attach_to = part_asset_data.get_editor_property(\"bAttachToSocket\")\r\n    except Exception:\r\n        attach_to = None        \r\n    if attach_to:\r\n        backpack_info[\"AttachToSocket\"] = attach_to\r\n        ##print(\"  AttachToSocket: {}\".format(attach_to))                                \r\n\r\n    backpack_info[\"GenderPermitted\"] = \"\"\r\n    try:            \r\n        gender_permitted = part_asset_data.get_editor_property(\"GenderPermitted\")\r\n    except Exception:\r\n        gender_permitted = None            \r\n    if gender_permitted:\r\n        display_name = gender_permitted.get_display_name()\r\n        backpack_info[\"GenderPermitted\"] = str(display_name)\r\n        ##print(\"  GenderPermitted: {}\".format(display_name)) \r\n\r\n    try:            \r\n        body_types_permitted = part_asset_data.get_editor_property(\"BodyTypesPermitted\")\r\n    except Exception:\r\n        body_types_permitted = None            \r\n    if body_types_permitted:\r\n        display_name = body_types_permitted.get_display_name()\r\n        backpack_info[\"BodyTypesPermitted\"] = str(display_name)\r\n        ##print(\"  BodyTypesPermitted: {}\".format(display_name))                                \r\n\r\n    try:\r\n        attach_over_data = part_asset_data.get_editor_property(\"AdditionalData\").get_editor_property(\"AttachmentOverrideData\")\r\n    except Exception:\r\n        attach_over_data = None\r\n    if attach_over_data:\r\n        backpack_info[\"AttachmentOverrideData\"] = attach_over_data.get_name()\r\n        ##print(\"  AttachmentOverrideData: {}\".format(attach_over_data.get_name()))  \r\n\r\n    try:\r\n        collide_other_parts_cloth = part_asset_data.get_editor_property(\"AdditionalData\").get_editor_property(\"bCollideWithOtherPartsCloth\")\r\n        backpack_info[\"CollideWithOtherPartsCloth\"] = collide_other_parts_cloth\r\n        ##print(\"  CollideWithOtherPartsCloth: {}\".format(collide_other_parts_cloth))                      \r\n    except:\r\n        print(\"Failed to get collide parts\")\r\n\r\n    backpack_info[\"UseClothCollisionFromOtherParts\"] = \"\"\r\n    try:\r\n        use_collision_other_parts = part_asset_data.get_editor_property(\"AdditionalData\").get_editor_property(\"bUseClothCollisionFromOtherParts\")\r\n        backpack_info[\"UseClothCollisionFromOtherParts\"] = use_collision_other_parts\r\n        ##print(\"  UseClothCollisionFromOtherParts: {}\".format(use_collision_other_parts))                      \r\n    except:\r\n        print(\"Failed to get use collision parts\")    \r\n            \r\n    return backpack_info\r\n\r\ndef get_backpack_data(bid_asset):\r\n    \"\"\"\r\n    Get backpack data from a BIP asset\r\n    \r\n    bid_asset - \"AthenaBackpackItemDefinition\"\r\n    \"\"\"    \r\n\r\n    cp_name = unreal.Name(\"CharacterPartType\")      \r\n    asset_data = unreal.AssetData.get_asset(bid_asset)    \r\n    try:    \r\n        character_parts = asset_data.get_editor_property(\"CharacterParts\")\r\n    except Exception:\r\n        character_parts = []\r\n        print('  Failed to get property \"CharacterParts\" for this BID: {}'.format(bid_asset))\r\n        return\r\n\r\n    ## get character part info for this backpack\r\n    for part in character_parts:\r\n        if not part:\r\n            continue\r\n            \r\n        if part.get_class().get_name() == \"CustomCharacterPart\":\r\n            part_asset = asset_registry.get_asset_by_object_path(part.get_path_name())               \r\n            part_type = part_asset.get_tag_value(cp_name)\r\n            if part_type == \"Backpack\":\r\n\r\n                ## only print if cape\r\n                gameplay_tags = get_gamplay_tags(bid_asset)\r\n                backpack_part_data = get_part_data(part_asset)\r\n                if not backpack_part_data:\r\n                    continue\r\n                \r\n                backpack_is_cape = False\r\n                if gameplay_tags:                    \r\n                    if \"Cosmetics.Backpack.IsCape\" in gameplay_tags:\r\n                        backpack_is_cape = True\r\n                if \"AttachSocketName\" in backpack_part_data:\r\n                    if backpack_part_data[\"AttachSocketName\"] == \"cape\":\r\n                        backpack_is_cape = True\r\n                if \"LootSocket\" in backpack_part_data:                    \r\n                    if backpack_part_data[\"LootSocket\"] == \"cape\":\r\n                        backpack_is_cape = True\r\n                if \"UseClothCollisionFromOtherParts\" in backpack_part_data:\r\n                    if backpack_part_data[\"UseClothCollisionFromOtherParts\"]:\r\n                        backpack_is_cape = True\r\n                    \r\n                if backpack_is_cape:                    \r\n                    ##print(\"\\nBID Asset: {}\".format(bid_asset.asset_name))\r\n                    ##print(\"  CharacterPart: {}\".format(part.get_name()))                   \r\n                    ##for data, value in backpack_part_data.items():\r\n                        ##print(\"  {}: {}\".format(data, value))\r\n                    ##for tag in gameplay_tags:\r\n                        ##print(\"  Gameplay Tag: {}\".format(tag))\r\n                        \r\n                    cape_data = backpack_part_data\r\n                    cape_data[\"CharacterPart\"] = part_asset.asset_name\r\n                    if gameplay_tags:\r\n                        cape_data[\"GameplayTags\"] = gameplay_tags\r\n                        \r\n                    return cape_data\r\n    \r\n    return None\r\n\r\ndef main():\r\n    \"\"\"\r\n    Get all BID assets\r\n    Get GameplayTags and CharacterPart info\r\n    \"\"\"\r\n    \r\n    bid_name = unreal.Name(\"FortniteGame.AthenaBackpackItemDefinition\")\r\n    bid_filter = unreal.ARFilter(class_names=[bid_name])\r\n    bid_assets = asset_registry.get_assets(bid_filter)\r\n    print(\"Number of BID assets: {}\".format(len(bid_assets)))\r\n    \r\n    ##bid_index = 0\r\n    all_cape_data = []\r\n    for bid_asset in bid_assets:\r\n        cape_data = get_backpack_data(bid_asset)\r\n        if cape_data:\r\n            cape_data[\"BID\"] = bid_asset.asset_name\r\n            all_cape_data.append(cape_data)\r\n            ##all_cape_data[bid_asset] = cape_data\r\n        \r\n        ## return early\r\n        ##bid_index += 1        \r\n        ##if bid_index > 100:            \r\n            ##break\r\n        \r\n    ## header list\r\n    csv_header = [\"BID\", \"CharacterPart\", \"AnimBP\", \"AttachSocketName\", \"AttachToSocket\", \"LootSocket\", \"CollideWithOtherPartsCloth\", \"UseClothCollisionFromOtherParts\",\r\n                  \"GenderPermitted\", \"BodyTypesPermitted\", \"AttachmentOverrideData\", \"GameplayTags\"]\r\n    \r\n    ## write to csv\r\n    temp_csv = r\"d:\\backpack_info_sheet.csv\"\r\n    with open(temp_csv, \"w\") as file:\r\n        writer = csv.DictWriter(file, fieldnames=csv_header)\r\n        writer.writeheader()\r\n        writer.writerows(all_cape_data)\r\n        \r\n    print(\"Number of Capes: {}\".format(len(all_cape_data)))\r\n    \r\n \r\ndef get_selected_bid_data():\r\n    \"\"\"\r\n    Get BID data for selected BIDs\r\n    \"\"\"\r\n    selection = unreal.EditorUtilityLibrary.get_selected_asset_data()\r\n    for sel in selection:\r\n        asset_class = sel.asset_class_path.asset_name\r\n        if not asset_class == \"AthenaBackpackItemDefinition\":\r\n            continue\r\n        get_backpack_data(sel)\r\n        \r\n        \r\nmain()")
   > Set IsRunning(IsRunning=true)
   > Set Text(Text="LOCTEXT(\"Finished!\")", self=Get Get_Info_Status())
   > Set Text(Text="LOCTEXT(\"File Location : \\\"D:\\\\backpack_info_sheet.csv\\\"\")", self=Get File_Location_Text())
   > Set Text(Text="LOCTEXT(\"Open Google Sheets > File Import > This csv file\")", self=Get Info_Text())
   Event Construct()
   > SetObject(self=Get Get_Info_DetailsView() -> (Get_Info_DetailsView:Details View Object Reference), NewObject=Self-Reference())
   Event PreConstruct()
   Event Tick()
  ]}]
variables [IsRunning {type=boolean}]
widgets {
 info {rootWidgetClass=&ScrollBox, widgetCount=8}
 widgets [
  ScrollBox_0 {
   widgetClassPath=&ScrollBox
   widget=ScrollBox_0
   properties {
    widgetBarStyle {
     thickness=8
     draggedThumbImage {
      bIsDynamicallyLoaded=false
      drawAs=RoundedBox
      imageType=NoImage
      mirroring=NoMirror
      resourceName=None
      resourceObject=null
      tiling=NoTile
      imageSize=<0 0>
      margin=<ltrb 0 0 0 0>
      outlineSettings {
       bUseBrushTransparency=false
       roundingType=FixedRadius
       width=0
       color {colorUseRule=UseColor_Specified, specifiedColor=<rgba 0 0 0 0>}
       cornerRadii=<4 4 4 4>
      }
      tintColor {colorUseRule=UseColor_Specified, specifiedColor=<rgba 0.2159 0.2159 0.2159 1>}
      uVRegion {bIsValid=false, max=<0 0>, min=<0 0>}
     }
     horizontalBackgroundImage {
      bIsDynamicallyLoaded=false
      drawAs=NoDrawType
      imageType=NoImage
      mirroring=NoMirror
      resourceName=None
      resourceObject=null
      tiling=NoTile
      imageSize=<0 0>
      margin=<ltrb 0 0 0 0>
      outlineSettings {
       bUseBrushTransparency=false
       roundingType=HalfHeightRadius
       width=0
       color {colorUseRule=UseColor_Specified, specifiedColor=<rgba 0 0 0 0>}
       cornerRadii=<0 0 0 0>
      }
      tintColor {colorUseRule=UseColor_Specified, specifiedColor=<rgba 1 1 1 1>}
      uVRegion {bIsValid=false, max=<0 0>, min=<0 0>}
     }
     horizontalBottomSlotImage {
      bIsDynamicallyLoaded=false
      drawAs=NoDrawType
      imageType=NoImage
      mirroring=NoMirror
      resourceName=None
      resourceObject=null
      tiling=NoTile
      imageSize=<0 0>
      margin=<ltrb 0 0 0 0>
      outlineSettings {
       bUseBrushTransparency=false
       roundingType=HalfHeightRadius
       width=0
       color {colorUseRule=UseColor_Specified, specifiedColor=<rgba 0 0 0 0>}
       cornerRadii=<0 0 0 0>
      }
      tintColor {colorUseRule=UseColor_Specified, specifiedColor=<rgba 1 1 1 1>}
      uVRegion {bIsValid=false, max=<0 0>, min=<0 0>}
     }
     horizontalTopSlotImage {
      bIsDynamicallyLoaded=false
      drawAs=NoDrawType
      imageType=NoImage
      mirroring=NoMirror
      resourceName=None
      resourceObject=null
      tiling=NoTile
      imageSize=<0 0>
      margin=<ltrb 0 0 0 0>
      outlineSettings {
       bUseBrushTransparency=false
       roundingType=HalfHeightRadius
       width=0
       color {colorUseRule=UseColor_Specified, specifiedColor=<rgba 0 0 0 0>}
       cornerRadii=<0 0 0 0>
      }
      tintColor {colorUseRule=UseColor_Specified, specifiedColor=<rgba 1 1 1 1>}
      uVRegion {bIsValid=false, max=<0 0>, min=<0 0>}
     }
     hoveredThumbImage {
      bIsDynamicallyLoaded=false
      drawAs=RoundedBox
      imageType=NoImage
      mirroring=NoMirror
      resourceName=None
      resourceObject=null
      tiling=NoTile
      imageSize=<0 0>
      margin=<ltrb 0 0 0 0>
      outlineSettings {
       bUseBrushTransparency=false
       roundingType=FixedRadius
       width=0
       color {colorUseRule=UseColor_Specified, specifiedColor=<rgba 0 0 0 0>}
       cornerRadii=<4 4 4 4>
      }
      tintColor {colorUseRule=UseColor_Specified, specifiedColor=<rgba 0.2159 0.2159 0.2159 1>}
      uVRegion {bIsValid=false, max=<0 0>, min=<0 0>}
     }
     normalThumbImage {
      bIsDynamicallyLoaded=false
      drawAs=RoundedBox
      imageType=NoImage
      mirroring=NoMirror
      resourceName=None
      resourceObject=null
      tiling=NoTile
      imageSize=<0 0>
      margin=<ltrb 0 0 0 0>
      outlineSettings {
       bUseBrushTransparency=false
       roundingType=FixedRadius
       width=0
       color {colorUseRule=UseColor_Specified, specifiedColor=<rgba 0 0 0 0>}
       cornerRadii=<4 4 4 4>
      }
      tintColor {colorUseRule=UseColor_Specified, specifiedColor=<rgba 0.0953 0.0953 0.0953 1>}
      uVRegion {bIsValid=false, max=<0 0>, min=<0 0>}
     }
     verticalBackgroundImage {
      bIsDynamicallyLoaded=false
      drawAs=NoDrawType
      imageType=NoImage
      mirroring=NoMirror
      resourceName=None
      resourceObject=null
      tiling=NoTile
      imageSize=<0 0>
      margin=<ltrb 0 0 0 0>
      outlineSettings {
       bUseBrushTransparency=false
       roundingType=HalfHeightRadius
       width=0
       color {colorUseRule=UseColor_Specified, specifiedColor=<rgba 0 0 0 0>}
       cornerRadii=<0 0 0 0>
      }
      tintColor {colorUseRule=UseColor_Specified, specifiedColor=<rgba 1 1 1 1>}
      uVRegion {bIsValid=false, max=<0 0>, min=<0 0>}
     }
     verticalBottomSlotImage {
      bIsDynamicallyLoaded=false
      drawAs=NoDrawType
      imageType=NoImage
      mirroring=NoMirror
      resourceName=None
      resourceObject=null
      tiling=NoTile
      imageSize=<0 0>
      margin=<ltrb 0 0 0 0>
      outlineSettings {
       bUseBrushTransparency=false
       roundingType=HalfHeightRadius
       width=0
       color {colorUseRule=UseColor_Specified, specifiedColor=<rgba 0 0 0 0>}
       cornerRadii=<0 0 0 0>
      }
      tintColor {colorUseRule=UseColor_Specified, specifiedColor=<rgba 1 1 1 1>}
      uVRegion {bIsValid=false, max=<0 0>, min=<0 0>}
     }
     verticalTopSlotImage {
      bIsDynamicallyLoaded=false
      drawAs=NoDrawType
      imageType=NoImage
      mirroring=NoMirror
      resourceName=None
      resourceObject=null
      tiling=NoTile
      imageSize=<0 0>
      margin=<ltrb 0 0 0 0>
      outlineSettings {
       bUseBrushTransparency=false
       roundingType=HalfHeightRadius
       width=0
       color {colorUseRule=UseColor_Specified, specifiedColor=<rgba 0 0 0 0>}
       cornerRadii=<0 0 0 0>
      }
      tintColor {colorUseRule=UseColor_Specified, specifiedColor=<rgba 1 1 1 1>}
      uVRegion {bIsValid=false, max=<0 0>, min=<0 0>}
     }
    }
    widgetStyle {
     barThickness=8
     bottomShadowBrush {
      bIsDynamicallyLoaded=false
      drawAs=Box
      imageType=FullColor
      mirroring=NoMirror
      resourceName=../../../Engine/Content/Slate/Common/ScrollBoxShadowBottom.png
      resourceObject=null
      tiling=NoTile
      imageSize=<16 8>
      margin=<ltrb 0.5 0 0.5 1>
      outlineSettings {
       bUseBrushTransparency=false
       roundingType=HalfHeightRadius
       width=0
       color {colorUseRule=UseColor_Specified, specifiedColor=<rgba 0 0 0 0>}
       cornerRadii=<0 0 0 0>
      }
      tintColor {colorUseRule=UseColor_Specified, specifiedColor=<rgba 1 1 1 1>}
      uVRegion {bIsValid=false, max=<0 0>, min=<0 0>}
     }
     horizontalScrolledContentPadding=<ltrb 0 0 1 0>
     leftShadowBrush {
      bIsDynamicallyLoaded=false
      drawAs=Box
      imageType=FullColor
      mirroring=NoMirror
      resourceName=../../../Engine/Content/Slate/Common/ScrollBoxShadowLeft.png
      resourceObject=null
      tiling=NoTile
      imageSize=<8 16>
      margin=<ltrb 1 0.5 0 0.5>
      outlineSettings {
       bUseBrushTransparency=false
       roundingType=HalfHeightRadius
       width=0
       color {colorUseRule=UseColor_Specified, specifiedColor=<rgba 0 0 0 0>}
       cornerRadii=<0 0 0 0>
      }
      tintColor {colorUseRule=UseColor_Specified, specifiedColor=<rgba 1 1 1 1>}
      uVRegion {bIsValid=false, max=<0 0>, min=<0 0>}
     }
     rightShadowBrush {
      bIsDynamicallyLoaded=false
      drawAs=Box
      imageType=FullColor
      mirroring=NoMirror
      resourceName=../../../Engine/Content/Slate/Common/ScrollBoxShadowRight.png
      resourceObject=null
      tiling=NoTile
      imageSize=<8 16>
      margin=<ltrb 0 0.5 1 0.5>
      outlineSettings {
       bUseBrushTransparency=false
       roundingType=HalfHeightRadius
       width=0
       color {colorUseRule=UseColor_Specified, specifiedColor=<rgba 0 0 0 0>}
       cornerRadii=<0 0 0 0>
      }
      tintColor {colorUseRule=UseColor_Specified, specifiedColor=<rgba 1 1 1 1>}
      uVRegion {bIsValid=false, max=<0 0>, min=<0 0>}
     }
     topShadowBrush {
      bIsDynamicallyLoaded=false
      drawAs=Box
      imageType=FullColor
      mirroring=NoMirror
      resourceName=../../../Engine/Content/Slate/Common/ScrollBoxShadowTop.png
      resourceObject=null
      tiling=NoTile
      imageSize=<16 8>
      margin=<ltrb 0.5 1 0.5 0>
      outlineSettings {
       bUseBrushTransparency=false
       roundingType=HalfHeightRadius
       width=0
       color {colorUseRule=UseColor_Specified, specifiedColor=<rgba 0 0 0 0>}
       cornerRadii=<0 0 0 0>
      }
      tintColor {colorUseRule=UseColor_Specified, specifiedColor=<rgba 1 1 1 1>}
      uVRegion {bIsValid=false, max=<0 0>, min=<0 0>}
     }
     verticalScrolledContentPadding=<ltrb 0 0 0 1>
    }
   }
  }
  Get_Infol_VerticalBox {widgetClassPath=&VerticalBox, parent=ScrollBox_0, widget=Get_Infol_VerticalBox, slot {class=ScrollBoxSlot}}
  Get_Info_DetailsView {
   widgetClassPath=&DetailsView
   parent=Get_Infol_VerticalBox
   widget=Get_Info_DetailsView
   properties {categoriesToShow [RenderOptions, RenderList]}
   slot {class=VerticalBoxSlot}
  }
  Get_Info_Button {
   widgetClassPath=&Button
   parent=Get_Infol_VerticalBox
   widget=Get_Info_Button
   properties {widgetStyle {
     clickedSlateSound {resourceObject=null}
     disabled {
      bIsDynamicallyLoaded=false
      drawAs=RoundedBox
      imageType=NoImage
      mirroring=NoMirror
      resourceName=None
      resourceObject=null
      tiling=NoTile
      imageSize=<0 0>
      margin=<ltrb 0 0 0 0>
      outlineSettings {
       bUseBrushTransparency=false
       roundingType=FixedRadius
       width=1
       color {colorUseRule=UseColor_ColorTable, specifiedColor=<rgba 0 0 0 0>}
       cornerRadii=<4 4 4 4>
      }
      tintColor {colorUseRule=UseColor_Specified, specifiedColor=<rgba 0.0395 0.0395 0.0395 1>}
      uVRegion {bIsValid=false, max=<0 0>, min=<0 0>}
     }
     disabledForeground {colorUseRule=UseColor_ColorTable, specifiedColor=<rgba 0 0 0 0>}
     hovered {
      bIsDynamicallyLoaded=false
      drawAs=RoundedBox
      imageType=NoImage
      mirroring=NoMirror
      resourceName=None
      resourceObject=null
      tiling=NoTile
      imageSize=<0 0>
      margin=<ltrb 0 0 0 0>
      outlineSettings {
       bUseBrushTransparency=false
       roundingType=FixedRadius
       width=1
       color {colorUseRule=UseColor_ColorTable, specifiedColor=<rgba 0 0 0 0>}
       cornerRadii=<4 4 4 4>
      }
      tintColor {colorUseRule=UseColor_Specified, specifiedColor=<rgba 0.0953 0.0953 0.0953 1>}
      uVRegion {bIsValid=false, max=<0 0>, min=<0 0>}
     }
     hoveredForeground {colorUseRule=UseColor_ColorTable, specifiedColor=<rgba 0 0 0 0>}
     hoveredSlateSound {resourceObject=null}
     normal {
      bIsDynamicallyLoaded=false
      drawAs=RoundedBox
      imageType=NoImage
      mirroring=NoMirror
      resourceName=None
      resourceObject=null
      tiling=NoTile
      imageSize=<0 0>
      margin=<ltrb 0 0 0 0>
      outlineSettings {
       bUseBrushTransparency=false
       roundingType=FixedRadius
       width=1
       color {colorUseRule=UseColor_ColorTable, specifiedColor=<rgba 0 0 0 0>}
       cornerRadii=<4 4 4 4>
      }
      tintColor {colorUseRule=UseColor_Specified, specifiedColor=<rgba 0.0395 0.0395 0.0395 1>}
      uVRegion {bIsValid=false, max=<0 0>, min=<0 0>}
     }
     normalForeground {colorUseRule=UseColor_ColorTable, specifiedColor=<rgba 0 0 0 0>}
     normalPadding=<ltrb 12 1 12 1>
     pressed {
      bIsDynamicallyLoaded=false
      drawAs=RoundedBox
      imageType=NoImage
      mirroring=NoMirror
      resourceName=None
      resourceObject=null
      tiling=NoTile
      imageSize=<0 0>
      margin=<ltrb 0 0 0 0>
      outlineSettings {
       bUseBrushTransparency=false
       roundingType=FixedRadius
       width=1
       color {colorUseRule=UseColor_ColorTable, specifiedColor=<rgba 0 0 0 0>}
       cornerRadii=<4 4 4 4>
      }
      tintColor {colorUseRule=UseColor_Specified, specifiedColor=<rgba 0.0284 0.0284 0.0284 1>}
      uVRegion {bIsValid=false, max=<0 0>, min=<0 0>}
     }
     pressedForeground {colorUseRule=UseColor_ColorTable, specifiedColor=<rgba 0 0 0 0>}
     pressedPadding=<ltrb 12 2 12 0>
     pressedSlateSound {resourceObject=null}
    }}
   slot {class=VerticalBoxSlot}
  }
  Get_Info_Text {
   widgetClassPath=&CommonTextBlock
   parent=Get_Info_Button
   widget=Get_Info_Text
   properties {
    text=Get BID Cape Info
    font {
     bForceMonospaced=false
     bMaterialIsStencil=false
     fontMaterial=null
     fontObject=&BurbankSmall
     letterSpacing=0
     monospacedWidth=1
     size=17
     skewAmount=0
     typefaceFontName=Medium
     outlineSettings {
      bApplyOutlineToDropShadows=false
      bMiteredCorners=false
      bSeparateFillAlpha=false
      outlineMaterial=null
      outlineSize=0
      outlineColor=<rgba 0 0 0 1>
     }
    }
    shadowOffset=<0 0>
   }
   slot {class=ButtonSlot, properties {padding=<ltrb 10 0 10 0>}}
  }
  Get_Info_Status {
   widgetClassPath=&UEFN_TextBlock
   bIsVariable=true
   parent=ScrollBox_0
   widget=Get_Info_Status
   slot {class=ScrollBoxSlot, properties {horizontalAlignment=HAlign_Center, verticalAlignment=VAlign_Center, padding=<ltrb 0 5 0 0>}}
  }
  File_Location_Text {
   widgetClassPath=&UEFN_TextBlock
   bIsVariable=true
   parent=ScrollBox_0
   widget=File_Location_Text
   properties {font {
     bForceMonospaced=false
     bMaterialIsStencil=false
     fontMaterial=null
     fontObject=&BurbankSmall
     letterSpacing=0
     monospacedWidth=1
     size=14
     skewAmount=0
     typefaceFontName=Black
     outlineSettings {
      bApplyOutlineToDropShadows=false
      bMiteredCorners=false
      bSeparateFillAlpha=false
      outlineMaterial=null
      outlineSize=0
      outlineColor=<rgba 0 0 0 1>
     }
    }}
   slot {class=ScrollBoxSlot, properties {horizontalAlignment=HAlign_Center, verticalAlignment=VAlign_Center, padding=<ltrb 0 30 0 0>}}
  }
  Info_Text {
   widgetClassPath=&UEFN_TextBlock
   bIsVariable=true
   parent=ScrollBox_0
   widget=Info_Text
   properties {font {
     bForceMonospaced=false
     bMaterialIsStencil=false
     fontMaterial=null
     fontObject=&BurbankSmall
     letterSpacing=0
     monospacedWidth=1
     size=12
     skewAmount=0
     typefaceFontName=Black
     outlineSettings {
      bApplyOutlineToDropShadows=false
      bMiteredCorners=false
      bSeparateFillAlpha=false
      outlineMaterial=null
      outlineSize=0
      outlineColor=<rgba 0 0 0 1>
     }
    }}
   slot {class=ScrollBoxSlot, properties {horizontalAlignment=HAlign_Center, verticalAlignment=VAlign_Center, padding=<ltrb 0 10 0 0>}}
  }
 ]
}
