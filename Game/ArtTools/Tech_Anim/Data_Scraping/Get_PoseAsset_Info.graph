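# Editor utility widget for tech animation that exports PoseAsset metadata to a CSV at a hardcoded D: drive path.
# Its button runs embedded Python over selected PoseAssets, or all in the asset registry if none are selected.
# Only assets on Fortnite_Base_Head_Export_Skeleton are recorded, including retarget source, base pose and pose names.
# Text blocks then report completion regardless of outcome, and IsRunning is never reset to false.
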EditorUtilityWidgetBlueprint /Game/ArtTools/Tech_Anim/Data_Scraping/Get_PoseAsset_Info
graphs [EventGraph {nodes [
   "On Clicked (Get_Info_Button)"() {is=title K2Node_ComponentBoundEvent}
   > Set IsRunning(IsRunning:bool=true) -> (Output_Get:bool)
   > Set Text(self=Get Info_Text() -> (Info_Text:Text Block Object Reference))
   > Set Text(self=Get File_Location_Text() -> (File_Location_Text:Text Block Object Reference))
   > Set Text(Text="LOCTEXT(\"Getting PoseAsset Info..\")", self=Get Get_Info_Status() -> (Get_Info_Status:Text Block Object Reference))
   > Execute Python Script(PythonScript="import unreal\r\nimport csv\r\nimport logging\r\n\r\nasset_registry = unreal.AssetRegistryHelpers.get_asset_registry()\r\nlogger = logging.getLogger()\r\n\r\nPOSEASSET_INFOS_FILE = r\"d:\\poseAsset_infos.csv\"\r\nCSV_HEADER = [\"PoseAsset\", \"RetargetSource\", \"RetargetSourceAsset\", \"AdditiveBasePose\",\r\n              \"NumPoses\", \"Skeleton\", \"PoseNames\"]\r\nEXPECTED_SKELETONS = [\"Fortnite_Base_Head_Export_Skeleton\"]\r\n\r\n\r\ndef write_file(poseAsset_infos):\r\n    \"\"\"\r\n    Write a csv file with poseAsset infos\r\n    \"\"\"   \r\n    \r\n    print(f\"Writing PoseAsset Infos to Disk -> {POSEASSET_INFOS_FILE}\")\r\n    temp_csv = POSEASSET_INFOS_FILE\r\n    with open(temp_csv, \"w\", newline='') as file:\r\n        writer = csv.DictWriter(file, fieldnames=CSV_HEADER)\r\n        writer.writeheader()\r\n        writer.writerows(poseAsset_infos)\r\n    print(\"Saved poseAsset infos file.\")    \r\n\r\ndef get_poseasset_info(pose_asset, debug=False): \r\n    \"\"\"\r\n    Get poseAsset info\r\n    \"\"\"\r\n    pose_asset_info = {}\r\n    skeleton = pose_asset.get_editor_property(\"skeleton\")                    \r\n    if skeleton.get_name() in EXPECTED_SKELETONS:\r\n        retarget_source = pose_asset.get_editor_property(\"retarget_source\")\r\n        additive_pose = pose_asset.get_editor_property(\"additive_pose\")\r\n        additive_base_pose = pose_asset.get_base_pose_name()\r\n        pose_names = pose_asset.get_pose_names() or []\r\n        retarget_source_asset = pose_asset.get_editor_property(\"retarget_source_asset\")\r\n        if retarget_source_asset:                        \r\n            pose_asset_info[\"RetargetSourceAsset\"] = retarget_source_asset.get_name()                        \r\n        pose_asset_name = pose_asset.get_name()\r\n        if debug:\r\n            print(f\"PoseAsset: {pose_asset_name}\")\r\n            print(f\"Skeleton: {skeleton.get_name()}\")\r\n            print(f\"Retarget Source: {retarget_source}\")\r\n            if retarget_source_asset:\r\n                if debug: print(\"Retarget Source Asset: {}\".format(retarget_source_asset.get_name()))\r\n            print(f\"Additive Pose: {additive_pose}\")\r\n            print(f\"Additive Base Pose: {additive_base_pose}\")\r\n    \r\n        pose_asset_info[\"PoseAsset\"] = pose_asset.get_name()\r\n        pose_asset_info[\"RetargetSource\"] = retarget_source\r\n        pose_asset_info[\"AdditiveBasePose\"] = additive_base_pose\r\n        pose_asset_info[\"Skeleton\"] = skeleton.get_name()\r\n        pose_asset_info[\"NumPoses\"] = len(pose_names)\r\n        pose_asset_info[\"PoseNames\"] = pose_names\r\n        \r\n    return pose_asset_info\r\n\r\ndef main(asset_type=\"PoseAsset\", debug=False):\r\n    \"\"\"\r\n    Find all poseAssets then get their info\r\n    \"\"\"\r\n    unreal_name = unreal.Name(asset_type)\r\n    asset_filter = unreal.ARFilter(class_names=[unreal_name])\r\n    assets = asset_registry.get_assets(asset_filter)\r\n    if not assets:\r\n        logger.error(\"0 {} Assets were not found!\".format(unreal_name))\r\n        return\r\n    \r\n    print(\"{} {} assets were found\".format(len(assets), asset_type))\r\n    print(\"Getting PoseAsset Infos...\")\r\n    \r\n    poseAsset_infos = []\r\n    for asset in assets:\r\n        asset_class = asset.asset_class_path.asset_name\r\n        if asset_class == unreal.Name(asset_type):\r\n            pose_asset = unreal.AssetData.get_asset(asset)\r\n            if pose_asset:\r\n                pose_asset_info = get_poseasset_info(pose_asset, debug=debug)\r\n                if pose_asset_info:\r\n                    poseAsset_infos.append(pose_asset_info)\r\n          \r\n    if poseAsset_infos:        \r\n        write_file(poseAsset_infos)       \r\n\r\ndef run(debug=False):    \r\n    \r\n    ## check for selected poseAssets in the Content Browser\r\n    asset_type = \"PoseAsset\"\r\n    valid_selection = False\r\n    selection = unreal.EditorUtilityLibrary.get_selected_asset_data()\r\n    if selection:\r\n        poseAsset_infos = []   \r\n        for asset in selection:\r\n            asset_class = asset.asset_class_path.asset_name\r\n            if asset_class == unreal.Name(asset_type):\r\n                pose_asset = unreal.AssetData.get_asset(asset)\r\n                if pose_asset:                    \r\n                    valid_selection = True\r\n                    pose_asset_info = get_poseasset_info(pose_asset, debug=True)\r\n                    if pose_asset_info:\r\n                        poseAsset_infos.append(pose_asset_info)               \r\n            else:\r\n                logger.error(\"{} - is not a valid {} class!\".format(asset.asset_name, asset_type))\r\n\r\n        if poseAsset_infos:            \r\n            write_file(poseAsset_infos)               \r\n            \r\n    ## Get all PoseAsset infos\r\n    if not valid_selection:\r\n        main(asset_type=asset_type)\r\n\r\nrun()")
   > Set IsRunning(IsRunning=true)
   > Set Text(Text="LOCTEXT(\"Getting PoseAsset Info, Finished!\")", self=Get Get_Info_Status())
   > Set Text(Text="LOCTEXT(\"File Location : \\\"D:\\\\poseAsset_infos.csv\\\"\")", self=Get File_Location_Text())
   > Set Text(Text="LOCTEXT(\"Open Google Sheets > File Import > This csv file\")", self=Get Info_Text())
   Event Construct()
   > SetObject(self=Get Get_Info_DetailsView() -> (Get_Info_DetailsView:Details View Object Reference), NewObject=Self-Reference())
   Event PreConstruct()
   Event Tick()
  ]}]
variables [IsRunning {type=boolean}]
widgets {
 info {rootWidgetClass=&ScrollBox, widgetCount=8}
 widgets [
  ScrollBox_0 {
   widgetClassPath=&ScrollBox
   widget=ScrollBox_0
   properties {
    widgetBarStyle {
     thickness=8
     draggedThumbImage {
      bIsDynamicallyLoaded=false
      drawAs=RoundedBox
      imageType=NoImage
      mirroring=NoMirror
      resourceName=None
      resourceObject=null
      tiling=NoTile
      imageSize=<0 0>
      margin=<ltrb 0 0 0 0>
      outlineSettings {
       bUseBrushTransparency=false
       roundingType=FixedRadius
       width=0
       color {colorUseRule=UseColor_Specified, specifiedColor=<rgba 0 0 0 0>}
       cornerRadii=<4 4 4 4>
      }
      tintColor {colorUseRule=UseColor_Specified, specifiedColor=<rgba 0.2159 0.2159 0.2159 1>}
      uVRegion {bIsValid=false, max=<0 0>, min=<0 0>}
     }
     horizontalBackgroundImage {
      bIsDynamicallyLoaded=false
      drawAs=NoDrawType
      imageType=NoImage
      mirroring=NoMirror
      resourceName=None
      resourceObject=null
      tiling=NoTile
      imageSize=<0 0>
      margin=<ltrb 0 0 0 0>
      outlineSettings {
       bUseBrushTransparency=false
       roundingType=HalfHeightRadius
       width=0
       color {colorUseRule=UseColor_Specified, specifiedColor=<rgba 0 0 0 0>}
       cornerRadii=<0 0 0 0>
      }
      tintColor {colorUseRule=UseColor_Specified, specifiedColor=<rgba 1 1 1 1>}
      uVRegion {bIsValid=false, max=<0 0>, min=<0 0>}
     }
     horizontalBottomSlotImage {
      bIsDynamicallyLoaded=false
      drawAs=NoDrawType
      imageType=NoImage
      mirroring=NoMirror
      resourceName=None
      resourceObject=null
      tiling=NoTile
      imageSize=<0 0>
      margin=<ltrb 0 0 0 0>
      outlineSettings {
       bUseBrushTransparency=false
       roundingType=HalfHeightRadius
       width=0
       color {colorUseRule=UseColor_Specified, specifiedColor=<rgba 0 0 0 0>}
       cornerRadii=<0 0 0 0>
      }
      tintColor {colorUseRule=UseColor_Specified, specifiedColor=<rgba 1 1 1 1>}
      uVRegion {bIsValid=false, max=<0 0>, min=<0 0>}
     }
     horizontalTopSlotImage {
      bIsDynamicallyLoaded=false
      drawAs=NoDrawType
      imageType=NoImage
      mirroring=NoMirror
      resourceName=None
      resourceObject=null
      tiling=NoTile
      imageSize=<0 0>
      margin=<ltrb 0 0 0 0>
      outlineSettings {
       bUseBrushTransparency=false
       roundingType=HalfHeightRadius
       width=0
       color {colorUseRule=UseColor_Specified, specifiedColor=<rgba 0 0 0 0>}
       cornerRadii=<0 0 0 0>
      }
      tintColor {colorUseRule=UseColor_Specified, specifiedColor=<rgba 1 1 1 1>}
      uVRegion {bIsValid=false, max=<0 0>, min=<0 0>}
     }
     hoveredThumbImage {
      bIsDynamicallyLoaded=false
      drawAs=RoundedBox
      imageType=NoImage
      mirroring=NoMirror
      resourceName=None
      resourceObject=null
      tiling=NoTile
      imageSize=<0 0>
      margin=<ltrb 0 0 0 0>
      outlineSettings {
       bUseBrushTransparency=false
       roundingType=FixedRadius
       width=0
       color {colorUseRule=UseColor_Specified, specifiedColor=<rgba 0 0 0 0>}
       cornerRadii=<4 4 4 4>
      }
      tintColor {colorUseRule=UseColor_Specified, specifiedColor=<rgba 0.2159 0.2159 0.2159 1>}
      uVRegion {bIsValid=false, max=<0 0>, min=<0 0>}
     }
     normalThumbImage {
      bIsDynamicallyLoaded=false
      drawAs=RoundedBox
      imageType=NoImage
      mirroring=NoMirror
      resourceName=None
      resourceObject=null
      tiling=NoTile
      imageSize=<0 0>
      margin=<ltrb 0 0 0 0>
      outlineSettings {
       bUseBrushTransparency=false
       roundingType=FixedRadius
       width=0
       color {colorUseRule=UseColor_Specified, specifiedColor=<rgba 0 0 0 0>}
       cornerRadii=<4 4 4 4>
      }
      tintColor {colorUseRule=UseColor_Specified, specifiedColor=<rgba 0.0953 0.0953 0.0953 1>}
      uVRegion {bIsValid=false, max=<0 0>, min=<0 0>}
     }
     verticalBackgroundImage {
      bIsDynamicallyLoaded=false
      drawAs=NoDrawType
      imageType=NoImage
      mirroring=NoMirror
      resourceName=None
      resourceObject=null
      tiling=NoTile
      imageSize=<0 0>
      margin=<ltrb 0 0 0 0>
      outlineSettings {
       bUseBrushTransparency=false
       roundingType=HalfHeightRadius
       width=0
       color {colorUseRule=UseColor_Specified, specifiedColor=<rgba 0 0 0 0>}
       cornerRadii=<0 0 0 0>
      }
      tintColor {colorUseRule=UseColor_Specified, specifiedColor=<rgba 1 1 1 1>}
      uVRegion {bIsValid=false, max=<0 0>, min=<0 0>}
     }
     verticalBottomSlotImage {
      bIsDynamicallyLoaded=false
      drawAs=NoDrawType
      imageType=NoImage
      mirroring=NoMirror
      resourceName=None
      resourceObject=null
      tiling=NoTile
      imageSize=<0 0>
      margin=<ltrb 0 0 0 0>
      outlineSettings {
       bUseBrushTransparency=false
       roundingType=HalfHeightRadius
       width=0
       color {colorUseRule=UseColor_Specified, specifiedColor=<rgba 0 0 0 0>}
       cornerRadii=<0 0 0 0>
      }
      tintColor {colorUseRule=UseColor_Specified, specifiedColor=<rgba 1 1 1 1>}
      uVRegion {bIsValid=false, max=<0 0>, min=<0 0>}
     }
     verticalTopSlotImage {
      bIsDynamicallyLoaded=false
      drawAs=NoDrawType
      imageType=NoImage
      mirroring=NoMirror
      resourceName=None
      resourceObject=null
      tiling=NoTile
      imageSize=<0 0>
      margin=<ltrb 0 0 0 0>
      outlineSettings {
       bUseBrushTransparency=false
       roundingType=HalfHeightRadius
       width=0
       color {colorUseRule=UseColor_Specified, specifiedColor=<rgba 0 0 0 0>}
       cornerRadii=<0 0 0 0>
      }
      tintColor {colorUseRule=UseColor_Specified, specifiedColor=<rgba 1 1 1 1>}
      uVRegion {bIsValid=false, max=<0 0>, min=<0 0>}
     }
    }
    widgetStyle {
     barThickness=8
     bottomShadowBrush {
      bIsDynamicallyLoaded=false
      drawAs=Box
      imageType=FullColor
      mirroring=NoMirror
      resourceName=../../../Engine/Content/Slate/Common/ScrollBoxShadowBottom.png
      resourceObject=null
      tiling=NoTile
      imageSize=<16 8>
      margin=<ltrb 0.5 0 0.5 1>
      outlineSettings {
       bUseBrushTransparency=false
       roundingType=HalfHeightRadius
       width=0
       color {colorUseRule=UseColor_Specified, specifiedColor=<rgba 0 0 0 0>}
       cornerRadii=<0 0 0 0>
      }
      tintColor {colorUseRule=UseColor_Specified, specifiedColor=<rgba 1 1 1 1>}
      uVRegion {bIsValid=false, max=<0 0>, min=<0 0>}
     }
     horizontalScrolledContentPadding=<ltrb 0 0 1 0>
     leftShadowBrush {
      bIsDynamicallyLoaded=false
      drawAs=Box
      imageType=FullColor
      mirroring=NoMirror
      resourceName=../../../Engine/Content/Slate/Common/ScrollBoxShadowLeft.png
      resourceObject=null
      tiling=NoTile
      imageSize=<8 16>
      margin=<ltrb 1 0.5 0 0.5>
      outlineSettings {
       bUseBrushTransparency=false
       roundingType=HalfHeightRadius
       width=0
       color {colorUseRule=UseColor_Specified, specifiedColor=<rgba 0 0 0 0>}
       cornerRadii=<0 0 0 0>
      }
      tintColor {colorUseRule=UseColor_Specified, specifiedColor=<rgba 1 1 1 1>}
      uVRegion {bIsValid=false, max=<0 0>, min=<0 0>}
     }
     rightShadowBrush {
      bIsDynamicallyLoaded=false
      drawAs=Box
      imageType=FullColor
      mirroring=NoMirror
      resourceName=../../../Engine/Content/Slate/Common/ScrollBoxShadowRight.png
      resourceObject=null
      tiling=NoTile
      imageSize=<8 16>
      margin=<ltrb 0 0.5 1 0.5>
      outlineSettings {
       bUseBrushTransparency=false
       roundingType=HalfHeightRadius
       width=0
       color {colorUseRule=UseColor_Specified, specifiedColor=<rgba 0 0 0 0>}
       cornerRadii=<0 0 0 0>
      }
      tintColor {colorUseRule=UseColor_Specified, specifiedColor=<rgba 1 1 1 1>}
      uVRegion {bIsValid=false, max=<0 0>, min=<0 0>}
     }
     topShadowBrush {
      bIsDynamicallyLoaded=false
      drawAs=Box
      imageType=FullColor
      mirroring=NoMirror
      resourceName=../../../Engine/Content/Slate/Common/ScrollBoxShadowTop.png
      resourceObject=null
      tiling=NoTile
      imageSize=<16 8>
      margin=<ltrb 0.5 1 0.5 0>
      outlineSettings {
       bUseBrushTransparency=false
       roundingType=HalfHeightRadius
       width=0
       color {colorUseRule=UseColor_Specified, specifiedColor=<rgba 0 0 0 0>}
       cornerRadii=<0 0 0 0>
      }
      tintColor {colorUseRule=UseColor_Specified, specifiedColor=<rgba 1 1 1 1>}
      uVRegion {bIsValid=false, max=<0 0>, min=<0 0>}
     }
     verticalScrolledContentPadding=<ltrb 0 0 0 1>
    }
   }
  }
  Get_Infol_VerticalBox {widgetClassPath=&VerticalBox, parent=ScrollBox_0, widget=Get_Infol_VerticalBox, slot {class=ScrollBoxSlot}}
  Get_Info_DetailsView {
   widgetClassPath=&DetailsView
   parent=Get_Infol_VerticalBox
   widget=Get_Info_DetailsView
   properties {categoriesToShow [RenderOptions, RenderList]}
   slot {class=VerticalBoxSlot}
  }
  Get_Info_Button {
   widgetClassPath=&Button
   parent=Get_Infol_VerticalBox
   widget=Get_Info_Button
   properties {widgetStyle {
     clickedSlateSound {resourceObject=null}
     disabled {
      bIsDynamicallyLoaded=false
      drawAs=RoundedBox
      imageType=NoImage
      mirroring=NoMirror
      resourceName=None
      resourceObject=null
      tiling=NoTile
      imageSize=<0 0>
      margin=<ltrb 0 0 0 0>
      outlineSettings {
       bUseBrushTransparency=false
       roundingType=FixedRadius
       width=1
       color {colorUseRule=UseColor_ColorTable, specifiedColor=<rgba 0 0 0 0>}
       cornerRadii=<4 4 4 4>
      }
      tintColor {colorUseRule=UseColor_Specified, specifiedColor=<rgba 0.0395 0.0395 0.0395 1>}
      uVRegion {bIsValid=false, max=<0 0>, min=<0 0>}
     }
     disabledForeground {colorUseRule=UseColor_ColorTable, specifiedColor=<rgba 0 0 0 0>}
     hovered {
      bIsDynamicallyLoaded=false
      drawAs=RoundedBox
      imageType=NoImage
      mirroring=NoMirror
      resourceName=None
      resourceObject=null
      tiling=NoTile
      imageSize=<0 0>
      margin=<ltrb 0 0 0 0>
      outlineSettings {
       bUseBrushTransparency=false
       roundingType=FixedRadius
       width=1
       color {colorUseRule=UseColor_ColorTable, specifiedColor=<rgba 0 0 0 0>}
       cornerRadii=<4 4 4 4>
      }
      tintColor {colorUseRule=UseColor_Specified, specifiedColor=<rgba 0.0953 0.0953 0.0953 1>}
      uVRegion {bIsValid=false, max=<0 0>, min=<0 0>}
     }
     hoveredForeground {colorUseRule=UseColor_ColorTable, specifiedColor=<rgba 0 0 0 0>}
     hoveredSlateSound {resourceObject=null}
     normal {
      bIsDynamicallyLoaded=false
      drawAs=RoundedBox
      imageType=NoImage
      mirroring=NoMirror
      resourceName=None
      resourceObject=null
      tiling=NoTile
      imageSize=<0 0>
      margin=<ltrb 0 0 0 0>
      outlineSettings {
       bUseBrushTransparency=false
       roundingType=FixedRadius
       width=1
       color {colorUseRule=UseColor_ColorTable, specifiedColor=<rgba 0 0 0 0>}
       cornerRadii=<4 4 4 4>
      }
      tintColor {colorUseRule=UseColor_Specified, specifiedColor=<rgba 0.0395 0.0395 0.0395 1>}
      uVRegion {bIsValid=false, max=<0 0>, min=<0 0>}
     }
     normalForeground {colorUseRule=UseColor_ColorTable, specifiedColor=<rgba 0 0 0 0>}
     normalPadding=<ltrb 12 1 12 1>
     pressed {
      bIsDynamicallyLoaded=false
      drawAs=RoundedBox
      imageType=NoImage
      mirroring=NoMirror
      resourceName=None
      resourceObject=null
      tiling=NoTile
      imageSize=<0 0>
      margin=<ltrb 0 0 0 0>
      outlineSettings {
       bUseBrushTransparency=false
       roundingType=FixedRadius
       width=1
       color {colorUseRule=UseColor_ColorTable, specifiedColor=<rgba 0 0 0 0>}
       cornerRadii=<4 4 4 4>
      }
      tintColor {colorUseRule=UseColor_Specified, specifiedColor=<rgba 0.0284 0.0284 0.0284 1>}
      uVRegion {bIsValid=false, max=<0 0>, min=<0 0>}
     }
     pressedForeground {colorUseRule=UseColor_ColorTable, specifiedColor=<rgba 0 0 0 0>}
     pressedPadding=<ltrb 12 2 12 0>
     pressedSlateSound {resourceObject=null}
    }}
   slot {class=VerticalBoxSlot}
  }
  Get_Info_Text {
   widgetClassPath=&CommonTextBlock
   parent=Get_Info_Button
   widget=Get_Info_Text
   properties {
    text=Get PoseAsset Info
    font {
     bForceMonospaced=false
     bMaterialIsStencil=false
     fontMaterial=null
     fontObject=&BurbankSmall
     letterSpacing=0
     monospacedWidth=1
     size=17
     skewAmount=0
     typefaceFontName=Medium
     outlineSettings {
      bApplyOutlineToDropShadows=false
      bMiteredCorners=false
      bSeparateFillAlpha=false
      outlineMaterial=null
      outlineSize=0
      outlineColor=<rgba 0 0 0 1>
     }
    }
    shadowOffset=<0 0>
   }
   slot {class=ButtonSlot, properties {padding=<ltrb 10 0 10 0>}}
  }
  Get_Info_Status {
   widgetClassPath=&UEFN_TextBlock
   bIsVariable=true
   parent=ScrollBox_0
   widget=Get_Info_Status
   slot {class=ScrollBoxSlot, properties {horizontalAlignment=HAlign_Center, verticalAlignment=VAlign_Center, padding=<ltrb 0 5 0 0>}}
  }
  File_Location_Text {
   widgetClassPath=&UEFN_TextBlock
   bIsVariable=true
   parent=ScrollBox_0
   widget=File_Location_Text
   properties {font {
     bForceMonospaced=false
     bMaterialIsStencil=false
     fontMaterial=null
     fontObject=&BurbankSmall
     letterSpacing=0
     monospacedWidth=1
     size=14
     skewAmount=0
     typefaceFontName=Black
     outlineSettings {
      bApplyOutlineToDropShadows=false
      bMiteredCorners=false
      bSeparateFillAlpha=false
      outlineMaterial=null
      outlineSize=0
      outlineColor=<rgba 0 0 0 1>
     }
    }}
   slot {class=ScrollBoxSlot, properties {horizontalAlignment=HAlign_Center, verticalAlignment=VAlign_Center, padding=<ltrb 0 30 0 0>}}
  }
  Info_Text {
   widgetClassPath=&UEFN_TextBlock
   bIsVariable=true
   parent=ScrollBox_0
   widget=Info_Text
   properties {font {
     bForceMonospaced=false
     bMaterialIsStencil=false
     fontMaterial=null
     fontObject=&BurbankSmall
     letterSpacing=0
     monospacedWidth=1
     size=12
     skewAmount=0
     typefaceFontName=Black
     outlineSettings {
      bApplyOutlineToDropShadows=false
      bMiteredCorners=false
      bSeparateFillAlpha=false
      outlineMaterial=null
      outlineSize=0
      outlineColor=<rgba 0 0 0 1>
     }
    }}
   slot {class=ScrollBoxSlot, properties {horizontalAlignment=HAlign_Center, verticalAlignment=VAlign_Center, padding=<ltrb 0 10 0 0>}}
  }
 ]
}
